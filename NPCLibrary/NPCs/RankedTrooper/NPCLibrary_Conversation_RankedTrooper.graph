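# Conversation database for the Ranked Trooper NPC, a scout who talks about recruiting for the Elites.
# Its entry point leads to a speech node with a default line plus contextual greetings: a prioritized first-ever-interaction line and others gated by per-match conversation count checks on the trooper's NPC tag.
# The speech then branches via link nodes to shared NPC Library quests, basic services, and the Ranked Trooper's separate dialogue entry.
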ConversationDatabase /NPCLibrary/NPCs/RankedTrooper/NPCLibrary_Conversation_RankedTrooper
entry_points [{tag=Conversation.Entry.RankedTrooper, destinations [#1]}]
nodes [
 ConversationEntryPointNode_7 {
  id=#1
  class=ConversationEntryPointNode
  outputs [#2]
  properties {entryTag {tagName=Conversation.Entry.RankedTrooper}}
 }
 Speech_C_4 {
  id=#2
  class=Speech_C
  outputs [#3, #4, #5]
  properties {
   nodeName=""
   generalConfig {
    defaultMessage=It takes real excellence to get an invite to the Elites.
    contextualMessages [
     {
      message=Hey. You with the Elites?
      contextRequirements [{requirement {class=FortControllerRequirement_IsFirstEverConversationWithNPC}}]
      priority {curve {rowName=Tandem.Speech.Priority.FirstEverInteraction}}
     }
     {
      message=It takes real excellence to get an invite to the Elites.
      contextRequirements [{requirement {
         class=FortControllerRequirement_CheckNPCMatchConversationCount
         properties {nPCUniqueIdTag {tagName=AISpawnerData.Type.Specialist.Scout.RankedTrooper}}
        }}]
      priority {curve {rowName=None}}
     }
     {
      message="There's the Elites, and then there's everybody else. "
      contextRequirements [{requirement {
         class=FortControllerRequirement_CheckNPCMatchConversationCount
         properties {nPCUniqueIdTag {tagName=AISpawnerData.Type.Specialist.Scout.RankedTrooper}}
        }}]
      priority {curve {rowName=None}}
     }
     {
      message=I'm out here scouting for the best of the best. Maybe that's you.
      contextRequirements [{requirement {
         class=FortControllerRequirement_CheckNPCMatchConversationCount
         properties {nPCUniqueIdTag {tagName=AISpawnerData.Type.Specialist.Scout.RankedTrooper}}
        }}]
      priority {curve {rowName=None}}
     }
     {
      message=What are the Elites? Wouldn't you like to know...
      contextRequirements [{requirement {
         class=FortControllerRequirement_CheckNPCMatchConversationCount
         properties {nPCUniqueIdTag {tagName=AISpawnerData.Type.Specialist.Scout.RankedTrooper}}
        }}]
      priority {curve {rowName=None}}
     }
     {
      message=Real Elites know how to climb the Ranks.
      contextRequirements [{requirement {
         class=FortControllerRequirement_CheckNPCMatchConversationCount
         properties {nPCUniqueIdTag {tagName=AISpawnerData.Type.Specialist.Scout.RankedTrooper}}
        }}]
      priority {curve {rowName=None}}
     }
    ]
   }
  }
 }
 ConversationLinkNode_2 {
  id=#5
  class=ConversationLinkNode
  properties {remoteEntryTag {tagName=Conversation.Entry.RankedTrooper.Dialogue}}
 }
 ConversationLinkNode_0 {id=#4, class=ConversationLinkNode, properties {remoteEntryTag {tagName=Conversation.Entry.BasicServices}}}
 ConversationLinkNode_1 {id=#3, class=ConversationLinkNode, properties {remoteEntryTag {tagName=Conversation.Entry.NPCLibrary.Quests}}}
]
